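# UI material for the MurkyVole hover jets fuel meter, drawing a glowing frame and a radial fuel fill with alpha-composite blending.
# One texture supplies the frame mask in red and meter mask in green, each tinted from glow to fill color.
# FuelAmount drives MF_UI_RadialProgress, which masks the gradient meter in color and opacity.
# CameraVelocity, scaled and clamped to a maximum drift, shifts the UVs for parallax.
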Material /MurkyVoleGameplay/Gameplay/HoverJets/UI/M_UI_MurkyVole_HoverJets_FuelMeter
graph {
 properties {bAutomaticallySetUsageInEditor=false, blendMode=BLEND_AlphaComposite, materialDomain=MD_UI, usedShadingModels=Default Lit}
 nodes [
  n1 = VectorParameter() {parameterName=FrameGlow, defaultValue=<rgba 0 0.4115 0.3739 1>}
  n2 = VectorParameter() {parameterName=FrameFill, defaultValue=<rgba 1 1 1 1>}
  n3 = TextureSample(UVs=NamedRerouteUsage()) {texture=/MurkyVoleGameplay/Gameplay/HoverJets/UI/Textures/T_MurkyVole_HoverJet}
  n4 = SmoothStep(Value=@n3.R) {constMax=0.5}
  n5 = Multiply(A=ComponentMask(None=TextureCoordinate()) {g=true}) {constB=2}
  n6 = LinearInterpolate(
   A=VectorParameter() {parameterName=MeterGlow, defaultValue=<rgba 0 0.402 0.3663 1>}.RGB
   B=LinearInterpolate(
    A=Constant3Vector() {constant=<rgba 0.0056 0.8879 0.1812 1>}
    B=VectorParameter() {parameterName=MeterFill, defaultValue=<rgba 0.0056 0.9911 0.0648 1>}.RGB
    Alpha=Multiply(A=@n5, B=@n5)
   )
   Alpha=SmoothStep(Value=@n3.G) {constMax=0.5}
  )
  n7 = VectorParameter() {parameterName=Offset, defaultValue=<rgba 3.8749 0 0 1>}
  n8 = MaterialFunctionCall!(
   UVs=Add(A=NamedRerouteUsage(), B=ComponentMask(None=@n7.RGB) {g=true, r=true})
   Progress=ScalarParameter() {defaultValue=1, parameterName=FuelAmount}
   StartValue=ScalarParameter() {defaultValue=94.35, parameterName=Start}
   Range=ScalarParameter() {defaultValue=-8.5, parameterName=Range}
  ) {materialFunction=/FortUILibrary/MaterialFunctions/Basics/MF_UI_RadialProgress}
  Add(
   A=LinearInterpolate(B=LinearInterpolate(A=@n1.RGB, B=@n2.RGB, Alpha=@n4), Alpha=@n3.R)
   B=Multiply(A=LinearInterpolate(B=@n6, Alpha=@n3.G), B=@n8.Result)
  )
  LinearInterpolate(B=ScalarParameter() {defaultValue=0.75, parameterName=Tramslucency}, Alpha=Max(A=@n3.R, B=Multiply(A=@n3.G, B=@n8.Result)))
  n9 = Divide(A=ComponentMask(None=VectorParameter() {parameterName=CameraVelocity}.RGB) {g=true, r=true}, B=Constant2Vector() {g=0.84, r=9.6}) {desc="Decreases movement\r\nof element relative to others"}
  n10 = Constant2Vector() {desc=Max Drift, g=0.18, r=0.2}
  n11 = Multiply(A=@n10) {constB=-1}
  NamedRerouteDeclaration(None=Add(
    A=MaterialFunctionCall() {materialFunction=&GetUserInterfaceUV}.Normalized UV
    B=Clamp(None=@n9, Min=@n11, Max=@n10)
   )) {name=ParallaxUV, nodeColor=<rgba 0 0 0 1>}
 ]
}
