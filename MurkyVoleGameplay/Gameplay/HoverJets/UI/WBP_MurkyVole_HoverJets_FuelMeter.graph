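# HUD widget for MurkyVole hover jets that shows the viewed pawn's fuel on a material-driven meter with camera rotation effects.
# Async listeners track the Fuel attribute and regen tag, feeding the FuelAmount material parameter.
# A curve-table threshold selects looping Critical or LowFuel color animations (low is twice critical, capped), suppressed during regeneration.
# Delta-gated setters keep both states mutually exclusive, and finished animations restore cached default colors.
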WidgetBlueprint /MurkyVoleGameplay/Gameplay/HoverJets/UI/WBP_MurkyVole_HoverJets_FuelMeter
animations [
 Critical {
  binding_count=1
  length_seconds=0.25
  object=&self:Critical
  sequence {bindings [Meter {
     id=#1
     possessed_class=Image
     tracks [{
       type=MovieSceneWidgetMaterialTrack
       sections [{
         type=MovieSceneParameterSection
         end_frame=5
         start_frame=0
         channels [
          MeterFillR {type=MovieSceneScriptingFloatChannel, const [1,2]}
          MeterFillG {type=MovieSceneScriptingFloatChannel, keys [[2,0],[5,1]]}
          MeterFillB {type=MovieSceneScriptingFloatChannel, keys [[2,0],[5,1]]}
          MeterFillA {type=MovieSceneScriptingFloatChannel, const [1,1]}
          MeterGlowR {type=MovieSceneScriptingFloatChannel, const [1,2]}
          MeterGlowG {type=MovieSceneScriptingFloatChannel, keys [[2,0],[5,1]]}
          MeterGlowB {type=MovieSceneScriptingFloatChannel, keys [[2,0],[5,1]]}
          MeterGlowA {type=MovieSceneScriptingFloatChannel, const [1,1]}
         ]
        }]
      }]
    }]}
 }
 LowFuel {
  binding_count=1
  length_seconds=0.75
  object=&self:LowFuel
  sequence {bindings [Meter {
     id=#1
     possessed_class=Image
     tracks [{
       type=MovieSceneWidgetMaterialTrack
       sections [{
         type=MovieSceneParameterSection
         end_frame=15
         start_frame=0
         channels [
          MeterFillR {type=MovieSceneScriptingFloatChannel, keys [[0,1],[15,0]]}
          MeterFillG {type=MovieSceneScriptingFloatChannel, const [1,2]}
          MeterFillB {type=MovieSceneScriptingFloatChannel, keys [[0,0],[15,0.3419]]}
          MeterFillA {type=MovieSceneScriptingFloatChannel, const [1,1]}
          MeterGlowR {type=MovieSceneScriptingFloatChannel, keys [[0,1],[15,0]]}
          MeterGlowG {type=MovieSceneScriptingFloatChannel, keys [[0,1],[15,0.402]]}
          MeterGlowB {type=MovieSceneScriptingFloatChannel, keys [[0,0],[15,0.3663]]}
          MeterGlowA {type=MovieSceneScriptingFloatChannel, const [1,1]}
         ]
        }]
      }]
    }]}
 }
]
graphs [
 EventGraph {nodes [
   n1 = HandleFuelSet!() -> (OutputDelegate, then, FuelValue:double=~) {is=title K2Node_CustomEvent}
   n2 = "Clamp (Float)"(Value=float * float(A=@n1.FuelValue, B=0.01) {is=title K2Node_PromotableOperator}, Min=0, Max=1)
   > SetScalarParameterValue(self=Get MeterMAT() -> (MeterMAT:Material Instance Dynamic Object Reference), ParameterName=FuelAmount, Value=@n2)
   > n3 = Branch(Condition=Get IsFuelRegenerating() -> (IsFuelRegenerating:bool))
   > ResetFuelingAnimation()
   n4 = float * float(
    A=GetValueAtLevel!(Input:"Scalable Float Structure (by ref)"=Get LowFuelValue() -> (LowFuelValue:Scalable Float Structure), Level:float=0) -> (ReturnValue:float)
    B=0.01
   )
   n5 = Branch(execute=@n3.else, Condition="float <= float"(A=@n2, B=@n4) -> (ReturnValue=~))
   > SetIsFuelCritical!(execute, self, IsFuelCritical:bool=true) -> (then)
   n6 = float * float(A=@n4, B=2)
   n7 = Branch(execute=@n5.else, Condition="float <= float"(A=@n2, B="Clamp (Float)"(Value=@n6, Min=0, Max=0.4)) -> (ReturnValue=~))
   > SetIsFuelLow!(execute, self, IsFuelLow:bool=true) -> (then)
   ResetFuelingAnimation(execute=@n7.else)
   Event OnInitialized()
   n8 = Create Event()
   > BindToAnimationFinished(Animation=Get LowFuel() -> (LowFuel:Widget Animation Object Reference), Delegate=@n8)
   > BindToAnimationFinished(Animation=Get Critical() -> (Critical:Widget Animation Object Reference), Delegate=@n8)
   > n9 = GetDynamicMaterial(self=Get Meter() -> (Meter:Image Object Reference))
   > n10 = Set MeterMAT(MeterMAT:Material Instance Dynamic Object Reference=@n9) -> (Output_Get:Material Instance Dynamic Object Reference)
   > n11 = Get Vector Parameter Value!(execute, self:Material Instance Dynamic Object Reference=@n10, ParameterName:Name=Get MeterFillParamName() -> (MeterFillParamName:Name=None)) -> (then, ReturnValue:Linear Color Structure) {is=title K2Node_CallFunction}
   > Set DefaultMeterFillValue(DefaultMeterFillValue:Linear Color Structure=@n11) -> (Output_Get:Linear Color Structure)
   > n12 = Get Vector Parameter Value(self=Get MeterMAT(), ParameterName=Get MeterGlowParamName() -> (MeterGlowParamName:Name=None))
   > Set DefaultMeterGlowValue(DefaultMeterGlowValue:Linear Color Structure=@n12) -> (Output_Get:Linear Color Structure)
   Event Construct()
   > Parent: Construct()
   > n13 = Cast To FortPlayerPawn(Object=GetViewTarget!(self:Controller Object Reference=GetOwningPlayer()) -> (ReturnValue:Actor Object Reference))
   > n14 = Set Pawn(Pawn:Fort Player Pawn Object Reference=@n13) -> (Output_Get:Fort Player Pawn Object Reference)
   > Is Valid(InputObject=@n14)
   > HandleFuelSet!(
    execute
    self
    FuelValue:double=GetFloatAttribute!(
     Actor:Actor Object Reference=@n14
     Attribute:Gameplay Attribute Structure=(AttributeName="Fuel",Attribute=/Script/HoverGameplayRuntime.HoverAttributeSet:Fuel,AttributeOwner="&Class'&HoverAttributeSet'")
    ) -> (bSuccessfullyFoundAttribute:bool, ReturnValue:float).ReturnValue
   ) -> (then) {is=title K2Node_CallFunction}
   > n15 = WaitForAttributeChanged!(
    execute:Exec
    TargetActor:Actor Object Reference=Get Pawn() -> (Pawn:Fort Player Pawn Object Reference)
    Attribute:Gameplay Attribute Structure=(AttributeName="Fuel",Attribute=&HoverAttributeSet:Fuel,AttributeOwner="&Class'&HoverAttributeSet'")
    OnlyTriggerOnce:bool=false
   ) -> (
    then:Exec
    AsyncTaskProxy:Ability Async Wait Attribute Changed Object Reference
    Changed:Exec
    Attribute:Gameplay Attribute Structure
    NewValue:float=~
    OldValue:float=~
   ) {is=title K2Node_AsyncAction}
   > Set FuelSetListener(FuelSetListener:Ability Async Wait Attribute Changed Object Reference=@n15) -> (Output_Get:Ability Async Wait Attribute Changed Object Reference)
   > n16 = WaitGameplayTagAddToActor!(
    execute:Exec
    TargetActor:Actor Object Reference=Get Pawn()
    Tag:Gameplay Tag Structure=Get FuelRegenTag() -> (FuelRegenTag:Gameplay Tag Structure)
    OnlyTriggerOnce:bool=false
   ) -> (then:Exec, AsyncTaskProxy:Ability Async Wait Gameplay Tag Added Object Reference, Added:Exec) {is=title K2Node_AsyncAction}
   > Set FuelRegenListener(FuelRegenListener:Ability Async Wait Gameplay Tag Added Object Reference=@n16) -> (Output_Get:Ability Async Wait Gameplay Tag Added Object Reference)
   > n17 = WaitGameplayTagRemoveFromActor!(execute:Exec, TargetActor:Actor Object Reference=Get Pawn(), Tag:Gameplay Tag Structure=Get FuelRegenTag(), OnlyTriggerOnce:bool=false) -> (then:Exec, AsyncTaskProxy:Ability Async Wait Gameplay Tag Removed Object Reference, Removed:Exec) {is=title K2Node_AsyncAction}
   > Set FuelRegenRemovedListener(FuelRegenRemovedListener:Ability Async Wait Gameplay Tag Removed Object Reference=@n17) -> (Output_Get:Ability Async Wait Gameplay Tag Removed Object Reference)
   > n18 = Create UICamera Effects from Image List!(
    execute:Exec
    ImageList:Array of Image Object References=Make Array("[0]":Image Object Reference=Get Meter()) -> (Array:Array of Image Object References)
    PitchMaterial:Material Instance Dynamic Object Reference
   ) -> (then:Exec, CameraEffectsObject:"Fort Widget Camera Rotation Effects Object Reference (by ref)") {is=title K2Node_MacroInstance}
   > Set Camera Effects Object(Camera Effects Object:Fort Widget Camera Rotation Effects Object Reference=@n18) -> (Output_Get:Fort Widget Camera Rotation Effects Object Reference)
   Set IsFuelRegenerating(execute=@n17.Removed, IsFuelRegenerating:bool=false) -> (Output_Get:bool)
   Set IsFuelRegenerating(execute=@n16.Added, IsFuelRegenerating=true)
   HandleFuelSet(execute=@n15.Changed, FuelValue=@n15.NewValue)
   n19 = Event Tick()
   > Update!(
    execute
    self:Fort Widget Camera Rotation Effects Object Reference=Get Camera Effects Object() -> (Camera Effects Object:Fort Widget Camera Rotation Effects Object Reference)
    DeltaTime:float=@n19.InDeltaTime
   ) -> (then)
  ]}
 OnFuelAnimationFinished {nodes [
   OnFuelAnimationFinished() {is=title K2Node_FunctionEntry}
   n1 = Get MeterMAT()
   > SetVectorParameterValue(
    self=@n1
    ParameterName=Get MeterFillParamName() -> (MeterFillParamName=None)
    Value=Get DefaultMeterFillValue() -> (DefaultMeterFillValue:Linear Color Structure)
   )
   > SetVectorParameterValue(
    self=@n1
    ParameterName=Get MeterGlowParamName() -> (MeterGlowParamName=None)
    Value=Get DefaultMeterGlowValue() -> (DefaultMeterGlowValue:Linear Color Structure)
   )
  ]}
 ResetFuelingAnimation {nodes [
   ResetFuelingAnimation() {is=title K2Node_FunctionEntry}
   > SetIsFuelLow(IsFuelLow=false)
   > SetIsFuelCritical(IsFuelCritical=false)
  ]}
 SetIsFuelCritical {nodes [
   n1 = SetIsFuelCritical!() -> (then, IsFuelCritical:bool=~) {is=title K2Node_FunctionEntry}
   > n2 = Delta Gate Wildcard!(execute:Exec, Variable:"Boolean (by ref)"=Get IsFuelCritical() -> (IsFuelCritical:bool), New Value:bool=@n1) -> (Success:Exec, Updated Variable:"Boolean (by ref)", Previous Variable:bool=~) {is=title K2Node_MacroInstance}
   > n3 = Branch(Condition=@n2)
   > SetIsFuelLow(IsFuelLow=false)
   > PlayAnimation(InAnimation=Get Critical(), NumLoopsToPlay=0)
   StopAnimation(execute=@n3.else, InAnimation=Get Critical())
  ]}
 SetIsFuelLow {nodes [
   n1 = SetIsFuelLow!() -> (then, IsFuelLow:bool=~) {is=title K2Node_FunctionEntry}
   > n2 = Delta Gate Wildcard(Variable=Get IsFuelLow() -> (IsFuelLow:bool), New Value=@n1) -> (Previous Variable=~)
   > n3 = Branch(Condition=@n2)
   > SetIsFuelCritical(IsFuelCritical=false) {is=title K2Node_CallFunction}
   > PlayAnimation(InAnimation=Get LowFuel(), NumLoopsToPlay=0)
   StopAnimation(execute=@n3.else, InAnimation=Get LowFuel())
  ]}
]
variables [
 Camera Effects Object {type=FortWidgetCameraRotationEffects, default=null}
 DefaultMeterFillValue {type=LinearColor, default=<rgba 0 0 0 0>}
 DefaultMeterGlowValue {type=LinearColor, default=<rgba 0 0 0 0>}
 FuelRegenListener {type=AbilityAsync_WaitGameplayTagAdded, default=null}
 FuelRegenRemovedListener {type=AbilityAsync_WaitGameplayTagRemoved, default=null}
 FuelRegenTag {type=GameplayTag, default {tagName=GameplayCue.Athena.MurkyVole.Jets.FuelRegen}}
 FuelSetListener {type=AbilityAsync_WaitAttributeChanged, default=null}
 IsFuelCritical {type=boolean}
 IsFuelLow {type=boolean}
 IsFuelRegenerating {type=boolean}
 LowFuelValue {
  type=ScalableFloat
  default {registryType=None, value=1, curve {curveTable=/HoverGameplay/DataTables/HoverGameplayGameData, rowName=Default.Hover.Fx.LowFuelAmount}}
 }
 MeterFillParamName {type=string, default=MeterFill}
 MeterGlowParamName {type=string, default=MeterGlow}
 MeterMAT {type=MaterialInstanceDynamic}
 Pawn {type=FortPlayerPawn}
]
widgets {
 info {inheritedWidgetCount=1, parentClass=&FortHUDElementWidget, rootWidgetClass=&Image, widgetCount=2}
 widgets [
  Meter {
   widgetClassPath=&Image
   bIsVariable=true
   widget=Meter
   properties {
    visibility=SelfHitTestInvisible
    brush {
     bIsDynamicallyLoaded=false
     drawAs=Image
     imageType=FullColor
     mirroring=NoMirror
     resourceName=None
     resourceObject=&/M_UI_MurkyVole_HoverJets_FuelMeter_Inst
     tiling=NoTile
     imageSize=<128 512>
     margin=<ltrb 0 0 0 0>
     outlineSettings {
      bUseBrushTransparency=false
      roundingType=HalfHeightRadius
      width=0
      color {colorUseRule=UseColor_Specified, specifiedColor=<rgba 0 0 0 0>}
      cornerRadii=<0 0 0 0>
     }
     tintColor {colorUseRule=UseColor_Specified, specifiedColor=<rgba 1 1 1 1>}
     uVRegion {bIsValid=false, max=<0 0>, min=<0 0>}
    }
   }
  }
  TutorialHighlight {widgetClassPath=&TutorialHighlight}
 ]
}
